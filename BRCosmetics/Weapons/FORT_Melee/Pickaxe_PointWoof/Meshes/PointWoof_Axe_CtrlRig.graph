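# Control Rig for the PointWoof two-handed pickaxe that keeps its dog tag bone positioned at a fixed point on the hanging string chain.
# The forward solve reads the transform of dyn_DogTag_PosRef_null, a null under dyn_StringStart at the tag's rest offset.
# It writes only that translation to dyn_DogTag in global space, propagating to children and leaving the bone's rotation untouched.
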ControlRigBlueprint /BRCosmetics/Weapons/FORT_Melee/Pickaxe_PointWoof/Meshes/PointWoof_Axe_CtrlRig
graphs [
 RigVMFunctionLibrary {links [], nodes []}
 RigVMModel {nodes [
   RigUnit_BeginExecution = &RigUnit_BeginExecution()
   RigUnit_GetTransform = &RigUnit_GetTransform(Item=(Type=Null,Name="dyn_DogTag_PosRef_null"), bInitial=False)
   Set Transform = /Script/ControlRig.RigUnit_SetTranslation(
    Item=(Type=Bone,Name="dyn_DogTag")
    Space=GlobalSpace
    Weight=1
    bInitial=False
    bPropagateToChildren=True
    ExecutePin=@RigUnit_BeginExecution.ExecutePin
    Value=@RigUnit_GetTransform.Transform.Translation
   ) {is=struct RigVMUnitNode, template="Set Transform::Execute(in Item,in Space,in bInitial,in Value,in Weight,in bPropagateToChildren)"}
  ]}
]
hierarchy [
 root {loc=<0 0 0>, rot=<0 0 0 1>}
 b_AR_Root {parent=root, loc=<0 0 0>, rot=<0 0 0 1>}
 Melee_TwoHanded_Handle {parent=b_AR_Root, loc=<0 0 0>, rot=<0 0 0.7071 -0.7071>}
 base_StringStart {parent=Melee_TwoHanded_Handle, loc=<-0.3241 -2.1547 -8.7404>, rot=<0.4998 -0.5037 -0.5002 -0.4963>}
 dyn_StringSideRot {parent=base_StringStart, loc=<0 0 0>, rot=<0 0 0 1>}
 dyn_StringStart {parent=dyn_StringSideRot, loc=<2.3384 0 0>, rot=<0 0 0 1>}
 dyn_DogTag {parent=dyn_StringStart, loc=<4.4596 0.0154 0.0077>, rot=<0 0 0 1>}
 dyn_DogTag_PosRef_null {parent=dyn_StringStart, type=Null, loc=<4.4596 0.0154 0.0077>, rot=<0 0 0 1>}
 base_StaffBend {parent=Melee_TwoHanded_Handle, loc=<0 -0.315 18.5944>, rot=<0.525 -0.5301 0.4731 0.4685>}
 dyn_StaffBend {parent=base_StaffBend, loc=<0 0 0>, rot=<0 0 0 1>}
 body_root {parent=root, loc=<0 0 0>, rot=<0 0 0 1>}
 pelvis {parent=body_root, loc=<0 -0.5301 98.0571>, rot=<0.0127 0.707 -0.0127 -0.707>}
 spine_01 {parent=pelvis, loc=<6.1208 -0.9328 0>, rot=<0 0 0.034 -0.9994>}
 spine_02 {parent=spine_01, loc=<6.5997 -0.2095 0>, rot=<0 0 0.0238 -0.9997>}
 spine_03 {parent=spine_02, loc=<6.4914 -0.1237 0>, rot=<0 0 0.0149 -0.9999>}
 spine_04 {parent=spine_03, loc=<6.9598 0.0794 0>, rot=<0 0 -0.0132 -0.9999>}
 thigh_l {parent=pelvis, loc=<-1.0855 -0.2109 -8.7053>, rot=<0.0009 0.0355 0.0155 -0.9993>}
 calf_l {parent=thigh_l, loc=<-42.5284 -0.0019 -0.7297>, rot=<0.0006 -0.0161 -0.0391 -0.9991>}
 thigh_r {parent=pelvis, loc=<-1.0855 -0.2109 8.7053>, rot=<0.0355 -0.0009 0.9993 0.0155>}
 calf_r {parent=thigh_r, loc=<42.5284 0.0019 0.7297>, rot=<0.0006 -0.0161 -0.0391 -0.9991>}
]
