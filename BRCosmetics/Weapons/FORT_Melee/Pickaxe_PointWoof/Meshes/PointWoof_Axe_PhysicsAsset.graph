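# Physics asset for the PointWoof pickaxe, built around a dog tag hanging on a short string.
# A chain of constraints links the box-shaped tag through two small capsule string bones to base_StringStart, with limited swing and zero twist limits.
# Character leg capsules for both thighs and the right calf are also included, likely as collision for the tag, with only the right calf-to-thigh joint constrained, fully free in rotation.
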PhysicsAsset /BRCosmetics/Weapons/FORT_Melee/Pickaxe_PointWoof/Meshes/PointWoof_Axe_PhysicsAsset
bodies {
 base_StringStart [{shapeType=Capsule, length=0.3963, radius=0.505, shapeName=base_StringStart_capsule, center=<-0.8591 0 0>, rotation=<rot 85 -89.9997 0>}]
 calf_r [{shapeType=Capsule, length=32.5, radius=7.505, shapeName=calf_r_capsule, center=<18.2234 1.2767 0>, rotation=<rot 0 -90 90>}]
 dyn_DogTag [{shapeType=Box, extentX=6.5, extentY=2.5, extentZ=5.5, shapeName=dyn_DogTag_box, center=<3.8473 0 0>}]
 dyn_StringSideRot [{
   shapeType=Capsule
   length=0.0699
   radius=0.9749
   shapeName=dyn_StringSideRot_capsule
   center=<0.9545 0.066 -0.0188>
   rotation=<rot 0 179.5686 -89.9987>
  }]
 dyn_StringStart [{shapeType=Capsule, length=0.2749, radius=1.1454, shapeName=dyn_StringStart_capsule, center=<1.2058 -0.0035 0.0355>}]
 thigh_l [{
   shapeType=Capsule
   length=29.5295
   radius=8.32
   shapeName=thigh_l_capsule
   center=<-21.2385 0.0132 0.3997>
   rotation=<rot 0.0001 -90.0352 90>
  }]
 thigh_r [{
   shapeType=Capsule
   length=29.5295
   radius=8.32
   shapeName=thigh_l_capsule
   center=<20.4905 -0.0124 0.3997>
   rotation=<rot 0.0001 -90.0352 90>
  }]
}
constraints [
 {bone1Name=dyn_DogTag, bone2Name=dyn_StringStart, swing1LimitDegrees=70, swing2LimitDegrees=70, twistLimitDegrees=0}
 {bone1Name=dyn_StringStart, bone2Name=dyn_StringSideRot, swing1LimitDegrees=70, swing2LimitDegrees=40, twistLimitDegrees=0}
 {bone1Name=dyn_StringSideRot, bone2Name=base_StringStart, swing1LimitDegrees=2, swing2LimitDegrees=30, twistLimitDegrees=0}
 {
  bone1Name=calf_r
  bone2Name=thigh_r
  swing1LimitDegrees=45
  swing1Motion=Free
  swing2LimitDegrees=45
  swing2Motion=Free
  twistLimitDegrees=45
  twistMotion=Free
 }
]
