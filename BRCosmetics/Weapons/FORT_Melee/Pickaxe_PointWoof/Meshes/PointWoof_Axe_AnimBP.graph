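# Animation Blueprint for the PointWoof pickaxe that adds secondary motion to its bending staff, string and dog tag.
# The attach parent's pose is copied and layered, then passed through staff-bend AnimDynamics and body, pelvis and leg bone overrides.
# A bool blend picks between an AnimDynamics string chain and a RigidBody plus Control Rig branch.
# Outside the front end, the update event caches the parent body's leg, pelvis and root world transforms.
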AnimBlueprint /BRCosmetics/Weapons/FORT_Melee/Pickaxe_PointWoof/Meshes/PointWoof_Axe_AnimBP
target_skeleton=&/PointWoof_Axe_Skeleton
graphs [
 AnimGraph {nodes [
   Output Pose(Result=Use cached pose 'DynamicsCache'())
   n1 = Layered blend per bone(BasePose=Copy Pose From Mesh() {bUseAttachedParent=true}, BlendPoses_0=Copy Pose From Mesh()) {layerSetup [{branchFilters [{boneName=root}, {boneName=pelvis}, {boneName=thigh_l}, {boneName=calf_l}, {boneName=thigh_r}]}]}
   Save cached pose 'CRCache_Input'(Pose=@n1) {cacheName=CRCache_Input}
   n2 = Anim Dynamics - Bone: dyn_StringSideRot!(ComponentPose=Local To Component(LocalPose=Use cached pose 'JointTransforms_Cache'()), Alpha=1, GravityOverride:Vector="0,0,-980") -> (Pose) {
    angularDampingOverride=1
    angularSpringConstant=500
    bOverrideAngularDamping=true
    bUseGravityOverride=true
    bUsePlanarLimit=false
    simulationSpace=RootRelative
    alphaBoolBlend {blendOption=Linear}
    boundBone {boneName=dyn_StringSideRot}
    componentAppliedLinearAccClamp=<3000 3000 3000>
    componentLinearAccScale=<0.3 0.3 0.4>
    componentLinearVelScale=<0.3 0.3 0.4>
    gravityOverride {z=-980}
    physicsBodyDefinitions [{
      boundBone {boneName=dyn_StringSideRot}
      boxExtents=<6 12 12>
      constraintSetup {angularLimitsMax=<0 60 5>, angularLimitsMin=<0 -60 0>, angularTarget=<1 0 0>}
      localJointOffset=<2.5 0 0>
     }]
    relativeSpaceBone {boneName=dyn_StringStart}
   }
   n3 = Anim Dynamics - Bone: dyn_StringStart!(ComponentPose=@n2, Alpha=1, GravityOverride:Vector="0,0,-1500") -> (Pose) {
    angularDampingOverride=1
    angularSpringConstant=500
    bOverrideAngularDamping=true
    bUseGravityOverride=true
    bUsePlanarLimit=false
    simulationSpace=World
    alphaBoolBlend {blendOption=Linear}
    boundBone {boneName=dyn_StringStart}
    componentAppliedLinearAccClamp=<3000 3000 3000>
    componentLinearAccScale=<0.2 0.3 0.3>
    componentLinearVelScale=<0.2 0.3 0.3>
    gravityOverride {z=-980}
    physicsBodyDefinitions [{
      boundBone {boneName=dyn_StringStart}
      boxExtents=<10 20 20>
      constraintSetup {angularLimitsMax=<20 100 60>, angularLimitsMin=<-20 -100 -60>, angularTarget=<1 0 0>}
      localJointOffset=<3 0 0>
     }]
   }
   n4 = "Transform (Modify) Bone - Bone: dyn_StringSideRot"(ComponentPose=@n3, Alpha=1, Translation="0,-0.1,0", Rotation="0,0,-2", Scale="1,1.15,1") {
    rotationMode=BMM_Additive
    scaleMode=BMM_Additive
    translationMode=BMM_Additive
    alphaBoolBlend {blendOption=Linear}
    boneToModify {boneName=dyn_StringSideRot}
    rotation {roll=-2}
    scale {y=1.15}
    translation {y=-0.1}
   }
   n5 = "Transform (Modify) Bone - Bone: dyn_StringStart"(ComponentPose=@n4, Alpha=1, Translation="0,1,1", Rotation="0,0,15", Scale="1,1.3,1") {
    rotationMode=BMM_Additive
    scaleMode=BMM_Additive
    translationMode=BMM_Additive
    alphaBoolBlend {blendOption=Linear}
    boneToModify {boneName=dyn_StringStart}
    rotation {roll=15}
    scale {y=1.3}
    translation {y=1, z=1}
   }
   n6 = Anim Dynamics - Bone: dyn_DogTag!(ComponentPose=@n5, Alpha=1, GravityOverride:Vector="0,0,-980") -> (Pose) {
    angularDampingOverride=1
    angularSpringConstant=500
    bOverrideAngularDamping=true
    bUseGravityOverride=true
    bUsePlanarLimit=false
    simulationSpace=RootRelative
    alphaBoolBlend {blendOption=Linear}
    boundBone {boneName=dyn_DogTag}
    componentAppliedLinearAccClamp=<3000 3000 3000>
    componentLinearAccScale=<0.2 0.3 0.3>
    componentLinearVelScale=<0.2 0.3 0.3>
    gravityOverride {z=-980}
    physicsBodyDefinitions [{
      boundBone {boneName=dyn_DogTag}
      boxExtents=<8 15 15>
      constraintSetup {angularLimitsMax=<60 60 60>, angularLimitsMin=<-60 -60 -60>, angularTarget=<1 0 0>}
      localJointOffset=<3 0 0>
     }]
    relativeSpaceBone {boneName=dyn_StringStart}
   }
   n7 = "Transform (Modify) Bone - Bone: dyn_DogTag"(ComponentPose=@n6, Alpha=1, Translation="0,0,0", Rotation="0,0,0", Scale="1,0.9,1") {
    rotationMode=BMM_Additive
    scaleMode=BMM_Additive
    translationMode=BMM_Additive
    alphaBoolBlend {blendOption=Linear}
    boneToModify {boneName=dyn_DogTag}
    scale {y=0.9}
   }
   n8 = "Transform (Modify) Bone - Bone: base_StringStart"!(ComponentPose=Local To Component(LocalPose=Use cached pose 'JointTransforms_Cache'()), bAlphaBoolEnabled:bool=True, Rotation="-60,0,0") -> (Pose) {
    alphaInputType=Bool
    rotationMode=BMM_Additive
    translationMode=BMM_Additive
    alphaBoolBlend {blendOption=Linear}
    alphaScaleBiasClamp {bInterpResult=true}
    boneToModify {boneName=base_StringStart}
    rotation {pitch=-60}
   }
   n9 = RigidBody!(ComponentPose=@n8, bAlphaBoolEnabled:bool, OverrideWorldGravity:Vector="0,0,0") -> (Pose) {
    alphaInputType=Bool
    bOverrideWorldGravity=true
    overridePhysicsAsset=&/PointWoof_Axe_PhysicsAsset
    simulationSpace=BaseBoneSpace
    alphaBoolBlend {blendOption=Linear}
    componentAppliedLinearAccClamp=<3000 3000 3000>
    componentLinearAccScale=<0.7 0.4 0.4>
    componentLinearVelScale=<0.7 0.4 0.4>
    simSpaceSettings {
     maxAngularAcceleration=1000
     maxAngularVelocity=3
     maxLinearAcceleration=700
     maxLinearVelocity=500
     velocityScaleZ=0
     worldAlpha=0.35
     externalLinearDragV=<0.2 0.2 0.2>
    }
   }
   n10 = Control Rig(Source=Component To Local(ComponentPose=@n9)) {
    alphaBoolBlend {blendOption=Linear}
    controlRigAssetReference {blueprintRigClass=&/PointWoof_Axe_CtrlRig.PointWoof_Axe_CtrlRig_C}
   }
   Save cached pose 'DynamicsCache'(Pose=Blend Poses by bool(BlendPose_0=Component To Local(ComponentPose=@n7), BlendPose_1=@n10) {blendType=Linear}) {cacheName=DynamicsCache}
   n11 = Anim Dynamics - Bone: dyn_StaffBend!(
    ComponentPose=Local To Component(LocalPose=Use cached pose 'CRCache_Input'())
    bAlphaBoolEnabled:bool=True
    GravityOverride:Vector="0,0,-980"
   ) -> (Pose) {
    alphaInputType=Bool
    angularDampingOverride=1
    angularSpringConstant=20000
    bAngularSpring=true
    bOverrideAngularDamping=true
    bUseGravityOverride=true
    gravityScale=0.35
    simulationSpace=BoneRelative
    alphaBoolBlend {blendOption=Linear}
    boundBone {boneName=dyn_StaffBend}
    componentAppliedLinearAccClamp=<3000 3000 3000>
    componentLinearAccScale=<0.5 0.6 0.5>
    componentLinearVelScale=<0.5 0.6 0.5>
    gravityOverride {z=-980}
    physicsBodyDefinitions [{
      boundBone {boneName=dyn_StaffBend}
      boxExtents=<40 50 50>
      constraintSetup {angularLimitsMax=<0 30 30>, angularLimitsMin=<0 -30 -30>, angularTarget=<1 0 0>}
      localJointOffset=<3 0 0>
     }]
    relativeSpaceBone {boneName=b_AR_Root}
   }
   Save cached pose 'StickBend_Dynamics'(Pose=Component To Local(ComponentPose=@n11)) {cacheName=StickBend_Dynamics}
   n12 = "Transform (Modify) Bone - Bone: body_root"!(
    ComponentPose=Local To Component(LocalPose=Use cached pose 'StickBend_Dynamics'())
    bAlphaBoolEnabled:bool=True
    Translation="0,0,0"
    Rotation="0,0,0"
   ) -> (Pose) {
    alphaInputType=Bool
    rotationMode=BMM_Replace
    rotationSpace=BCS_WorldSpace
    scaleSpace=BCS_BoneSpace
    translationMode=BMM_Replace
    translationSpace=BCS_WorldSpace
    alphaBoolBlend {blendOption=Linear}
    alphaScaleBiasClamp {bInterpResult=true, interpSpeedDecreasing=1, interpSpeedIncreasing=1, inRange {max=0, min=-40}, outRange {max=0, min=1}}
    boneToModify {boneName=body_root}
   }
   n13 = "Transform (Modify) Bone - Bone: pelvis"!(ComponentPose=@n12, bAlphaBoolEnabled:bool=True, Translation="0,0,0", Rotation="0,0,0") -> (Pose) {
    alphaInputType=Bool
    rotationMode=BMM_Replace
    rotationSpace=BCS_WorldSpace
    scaleSpace=BCS_BoneSpace
    translationMode=BMM_Replace
    translationSpace=BCS_WorldSpace
    alphaBoolBlend {blendOption=Linear}
    alphaScaleBiasClamp {inRange {max=0, min=-40}, outRange {max=0, min=1}}
    boneToModify {boneName=pelvis}
   }
   n14 = "Transform (Modify) Bone - Bone: thigh_l"!(ComponentPose=@n13, bAlphaBoolEnabled:bool=True, Translation="0,0,0", Rotation="0,0,0") -> (Pose) {
    alphaInputType=Bool
    rotationMode=BMM_Replace
    rotationSpace=BCS_WorldSpace
    scaleSpace=BCS_BoneSpace
    translationMode=BMM_Replace
    translationSpace=BCS_WorldSpace
    alphaBoolBlend {blendOption=Linear}
    alphaScaleBiasClamp {inRange {max=0, min=-40}, outRange {max=0, min=1}}
    boneToModify {boneName=thigh_l}
   }
   n15 = "Transform (Modify) Bone - Bone: thigh_r"!(ComponentPose=@n14, bAlphaBoolEnabled:bool=True, Translation="0,0,0", Rotation="0,0,0") -> (Pose) {
    alphaInputType=Bool
    rotationMode=BMM_Replace
    rotationSpace=BCS_WorldSpace
    scaleSpace=BCS_BoneSpace
    translationMode=BMM_Replace
    translationSpace=BCS_WorldSpace
    alphaBoolBlend {blendOption=Linear}
    alphaScaleBiasClamp {inRange {max=0, min=-40}, outRange {max=0, min=1}}
    boneToModify {boneName=thigh_r}
   }
   n16 = "Transform (Modify) Bone - Bone: calf_l"!(ComponentPose=@n15, bAlphaBoolEnabled:bool=True, Translation="0,0,0", Rotation="0,0,0") -> (Pose) {
    alphaInputType=Bool
    rotationMode=BMM_Replace
    rotationSpace=BCS_WorldSpace
    scaleSpace=BCS_BoneSpace
    translationMode=BMM_Replace
    translationSpace=BCS_WorldSpace
    alphaBoolBlend {blendOption=Linear}
    alphaScaleBiasClamp {inRange {max=0, min=-40}, outRange {max=0, min=1}}
    boneToModify {boneName=calf_l}
   }
   n17 = "Transform (Modify) Bone - Bone: calf_r"!(ComponentPose=@n16, bAlphaBoolEnabled:bool=True, Translation="0,0,0", Rotation="0,0,0") -> (Pose) {
    alphaInputType=Bool
    rotationMode=BMM_Replace
    rotationSpace=BCS_WorldSpace
    scaleSpace=BCS_BoneSpace
    translationMode=BMM_Replace
    translationSpace=BCS_WorldSpace
    alphaBoolBlend {blendOption=Linear}
    alphaScaleBiasClamp {inRange {max=0, min=-40}, outRange {max=0, min=1}}
    boneToModify {boneName=calf_r}
   }
   Save cached pose 'JointTransforms_Cache'(Pose=Component To Local(ComponentPose=@n17)) {cacheName=JointTransforms_Cache}
  ]}
 EventGraph {nodes [
   Event BlueprintUpdateAnimation()
   > n1 = Gravity Override State Function DEPRECATED(CharacterPartAnimInstance=Self-Reference(), GameTimeInSeconds=GetGameTimeInSeconds(), GravityOverrideParams=Get Gravity Override Params())
   > Set GravityOverride(GravityOverride=@n1)
   > n2 = Branch(Condition=Get bIsInFrontEnd())
   n3 = GetPartSkeletalMeshComponent()
   Is Valid(exec=@n2.else, InputObject=@n3)
   n4 = GetAttachParent!(self:Scene Component Object Reference=@n3) -> (ReturnValue:Scene Component Object Reference)
   > Is Valid(InputObject=@n4)
   > Set Body_Mesh(Body_Mesh:Scene Component Object Reference=@n4) -> (Output_Get:Scene Component Object Reference)
   n5 = BreakTransform(InTransform=GetSocketTransform(self=Get Body_Mesh() -> (Body_Mesh:Scene Component Object Reference), InSocketName=pelvis, TransformSpace=RTS_World))
   > Set pelvis_location(pelvis_location:Vector=@n5) -> (Output_Get:Vector)
   > Set pelvis_rotation(pelvis_rotation:Rotator=@n5.Rotation) -> (Output_Get:Rotator)
   n6 = BreakTransform(InTransform=GetSocketTransform(self=Get Body_Mesh(), InSocketName=thigh_l, TransformSpace=RTS_World))
   > Set thigh_left_location(thigh_left_location:Vector=@n6) -> (Output_Get:Vector)
   > Set thigh_left_rotation(thigh_left_rotation:Rotator=@n6.Rotation) -> (Output_Get:Rotator)
   n7 = BreakTransform(InTransform=GetSocketTransform(self=Get Body_Mesh(), InSocketName=thigh_r, TransformSpace=RTS_World))
   > Set thigh_right_location(thigh_right_location:Vector=@n7) -> (Output_Get:Vector)
   > Set thigh_right_rotation(thigh_right_rotation:Rotator=@n7.Rotation) -> (Output_Get:Rotator)
   n8 = BreakTransform(InTransform=GetSocketTransform(self=Get Body_Mesh(), InSocketName=calf_l, TransformSpace=RTS_World))
   > Set calf_left_location(calf_left_location:Vector=@n8) -> (Output_Get:Vector)
   > Set calf_left_rotation(calf_left_rotation:Rotator=@n8.Rotation) -> (Output_Get:Rotator)
   n9 = BreakTransform(InTransform=GetSocketTransform(self=Get Body_Mesh(), InSocketName=calf_r, TransformSpace=RTS_World))
   > Set calf_right_location(calf_right_location:Vector=@n9) -> (Output_Get:Vector)
   > Set calf_right_rotation(calf_right_rotation:Rotator=@n9.Rotation) -> (Output_Get:Rotator)
   n10 = BreakTransform(InTransform=GetSocketTransform(self=Get Body_Mesh(), InSocketName=root, TransformSpace=RTS_World))
   > Set body_root_location(body_root_location:Vector=@n10) -> (Output_Get:Vector)
   > Set body_root_rotation(body_root_rotation:Rotator=@n10.Rotation) -> (Output_Get:Rotator)
   n11 = BreakTransform(InTransform=GetSocketTransform(self=Get Body_Mesh(), InSocketName=root, TransformSpace=RTS_World))
   > Set root_location(root_location:Vector=@n11) -> (Output_Get:Vector)
   > Set root_space_rotation(root_space_rotation:Rotator=@n11.Rotation) -> (Output_Get:Rotator)
  ]}
]
variables [
 Body_Mesh {type=SceneComponent, default=null}
 body_root_location {type=Vector}
 body_root_rotation {type=Rotator}
 calf_left_location {type=Vector}
 calf_left_rotation {type=Rotator}
 calf_right_location {type=Vector}
 calf_right_rotation {type=Rotator}
 Gravity Override Params {type=GravityOverrideParamsStruct}
 GravityOverride {type=Vector}
 isBodyAttached {type=boolean}
 pelvis_location {type=Vector}
 pelvis_rotation {type=Rotator}
 root_location {type=Vector}
 root_space_rotation {type=Rotator}
 Skel_Mesh {type=SceneComponent, default=null}
 thigh_left_location {type=Vector}
 thigh_left_rotation {type=Rotator}
 thigh_right_location {type=Vector}
 thigh_right_rotation {type=Rotator}
]
